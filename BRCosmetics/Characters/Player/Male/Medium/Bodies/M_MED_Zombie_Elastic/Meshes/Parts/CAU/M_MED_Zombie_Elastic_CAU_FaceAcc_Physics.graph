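# Physics asset for the face accessory mesh of the male medium Zombie Elastic (CAU) cosmetic body, providing capsule bodies for dynamic hair bones.
# A large head capsule anchors four separate hair bones, each constrained directly to the head with small angular limits.
# A main lock chains from a master bone through three segments with tight swing and twist limits, while the master's swing and twist are free relative to the head.
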PhysicsAsset /BRCosmetics/Characters/Player/Male/Medium/Bodies/M_MED_Zombie_Elastic/Meshes/Parts/CAU/M_MED_Zombie_Elastic_CAU_FaceAcc_Physics
bodies {
 dyn_hair_01 [{shapeType=Capsule, length=0, radius=1.2161, shapeName=dyn_hair_01_capsule, center=<1.6492 0 0>, rotation=<rot 90 0 0>}]
 dyn_hair_02 [{shapeType=Capsule, length=0, radius=1.2161, shapeName=dyn_hair_02_capsule, center=<1.1589 0 0>, rotation=<rot 90 0 0>}]
 dyn_hair_03 [{shapeType=Capsule, length=0, radius=1.2161, shapeName=dyn_hair_03_capsule, center=<1.5364 0 0>, rotation=<rot 90 0 0>}]
 dyn_hair_04 [{shapeType=Capsule, length=0, radius=1.2161, shapeName=dyn_hair_04_capsule, center=<1.1116 0 0>, rotation=<rot 90 0 0>}]
 dyn_hair_main_01 [{shapeType=Capsule, length=1.1598, radius=0.9713, shapeName=dyn_hair_main_01_capsule, center=<1.5 0 0>, rotation=<rot 90 0 0>}]
 dyn_hair_main_02 [{shapeType=Capsule, length=1.6379, radius=0.7538, shapeName=dyn_hair_main_02_capsule, center=<1.5 0 0>, rotation=<rot 90 0 0>}]
 dyn_hair_main_03 [{shapeType=Capsule, length=1.8181, radius=0.8646, shapeName=dyn_hair_main_03_capsule, center=<2 0 0>, rotation=<rot 90 0 0>}]
 dyn_hair_main_MSTR [{shapeType=Capsule, length=0.505, radius=0.505, shapeName=dyn_hair_main_MSTR_capsule, rotation=<rot 0 -90 0>}]
 head [{shapeType=Capsule, length=6.033, radius=5.9796, center=<8.1057 1.7961 -1.1069>, rotation=<rot 0 -5.0001 89.9999>}]
}
constraints [
 {
  bone1Name=dyn_hair_main_MSTR
  bone2Name=head
  swing1LimitDegrees=45
  swing1Motion=Free
  swing2LimitDegrees=45
  swing2Motion=Free
  twistLimitDegrees=45
  twistMotion=Free
 }
 {bone1Name=dyn_hair_main_01, bone2Name=dyn_hair_main_MSTR, swing1LimitDegrees=5, swing2LimitDegrees=15, twistLimitDegrees=3}
 {bone1Name=dyn_hair_main_02, bone2Name=dyn_hair_main_01, swing1LimitDegrees=5, swing2LimitDegrees=15, twistLimitDegrees=3}
 {bone1Name=dyn_hair_main_03, bone2Name=dyn_hair_main_02, swing1LimitDegrees=5, swing2LimitDegrees=15, twistLimitDegrees=5}
 {bone1Name=dyn_hair_04, bone2Name=head, swing1LimitDegrees=15, swing2LimitDegrees=15, twistLimitDegrees=3}
 {bone1Name=dyn_hair_03, bone2Name=head, swing1LimitDegrees=15, swing2LimitDegrees=15, twistLimitDegrees=3}
 {bone1Name=dyn_hair_02, bone2Name=head, swing1LimitDegrees=15, swing2LimitDegrees=15, twistLimitDegrees=3}
 {bone1Name=dyn_hair_01, bone2Name=head, swing1LimitDegrees=15, swing2LimitDegrees=15, twistLimitDegrees=3}
]
